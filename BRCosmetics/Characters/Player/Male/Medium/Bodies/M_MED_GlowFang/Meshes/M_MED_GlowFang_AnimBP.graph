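# Animation Blueprint for the GlowFang medium male player body cosmetic mesh, following the pose of its attached parent mesh.
# The copied pose runs through the GlowFang Control Rig, then through four chained Constraint nodes in component space.
# These position two extra knee cap bones per leg using weighted blends of calf, thigh, knee deform and twist bones, with weights mirrored left to right.
# The event graph holds no logic.
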AnimBlueprint /BRCosmetics/Characters/Player/Male/Medium/Bodies/M_MED_GlowFang/Meshes/M_MED_GlowFang_AnimBP
target_skeleton=&/M_MED_GlowFang_Skeleton
graphs [
 AnimGraph {nodes [
   n1 = Control Rig(Source=Copy Pose From Mesh() {bUseAttachedParent=true}) {
    alphaBoolBlend {blendOption=Linear}
    controlRigAssetReference {blueprintRigClass=&/M_MED_GlowFang_CtrlRig.M_MED_GlowFang_CtrlRig_C}
   }
   n2 = "Constraint - extra_knee_cap_1_l (PPPPPP)"!(
    ComponentPose=Local To Component(LocalPose=@n1)
    Alpha=1
    ConstraintWeights_0=0.4829
    ConstraintWeights_1:float=0.3055
    ConstraintWeights_2:float=0.1144
    ConstraintWeights_3:float=0.0631
    ConstraintWeights_4:float=0.0336
    ConstraintWeights_5:float=0.0005
   ) -> (Pose) {
    alphaBoolBlend {blendOption=Linear}
    boneToModify {boneName=extra_knee_cap_1_l}
    constraintSetup [
     {targetBone {boneName=calf_l}}
     {targetBone {boneName=thigh_l}}
     {targetBone {boneName=deform_knee_out_l}}
     {targetBone {boneName=deform_knee_in_l}}
     {targetBone {boneName=calf_twist_02_l}}
     {targetBone {boneName=thigh_twist_01_l}}
    ]
    constraintWeights [1,1,1,1,1,1]
   }
   n3 = "Constraint - extra_knee_cap_2_l (PPPPPP)"!(
    ComponentPose=@n2
    Alpha=1
    ConstraintWeights_0=0.4113
    ConstraintWeights_1:float=0.3262
    ConstraintWeights_2:float=0.1403
    ConstraintWeights_3:float=0.0913
    ConstraintWeights_4:float=0.026
    ConstraintWeights_5:float=0.0049
   ) -> (Pose) {
    alphaBoolBlend {blendOption=Linear}
    boneToModify {boneName=extra_knee_cap_2_l}
    constraintSetup [
     {targetBone {boneName=calf_l}}
     {targetBone {boneName=thigh_l}}
     {targetBone {boneName=deform_knee_out_l}}
     {targetBone {boneName=deform_knee_in_l}}
     {targetBone {boneName=calf_twist_02_l}}
     {targetBone {boneName=thigh_twist_01_l}}
    ]
    constraintWeights [1,1,1,1,1,1]
   }
   n4 = "Constraint - extra_knee_cap_2_r (PPPPPP)"!(
    ComponentPose=@n3
    Alpha=1
    ConstraintWeights_0=0.4113
    ConstraintWeights_1:float=0.3262
    ConstraintWeights_2:float=0.1403
    ConstraintWeights_3:float=0.0913
    ConstraintWeights_4:float=0.026
    ConstraintWeights_5:float=0.0049
   ) -> (Pose) {
    alphaBoolBlend {blendOption=Linear}
    boneToModify {boneName=extra_knee_cap_2_r}
    constraintSetup [
     {targetBone {boneName=calf_r}}
     {targetBone {boneName=thigh_r}}
     {targetBone {boneName=deform_knee_out_r}}
     {targetBone {boneName=deform_knee_in_r}}
     {targetBone {boneName=calf_twist_02_r}}
     {targetBone {boneName=thigh_twist_01_r}}
    ]
    constraintWeights [1,1,1,1,1,1]
   }
   n5 = "Constraint - extra_knee_cap_1_r (PPPPPP)"!(
    ComponentPose=@n4
    Alpha=1
    ConstraintWeights_0=0.4829
    ConstraintWeights_1:float=0.3055
    ConstraintWeights_2:float=0.1144
    ConstraintWeights_3:float=0.0631
    ConstraintWeights_4:float=0.0336
    ConstraintWeights_5:float=0.0005
   ) -> (Pose) {
    alphaBoolBlend {blendOption=Linear}
    boneToModify {boneName=extra_knee_cap_1_r}
    constraintSetup [
     {targetBone {boneName=calf_r}}
     {targetBone {boneName=thigh_r}}
     {targetBone {boneName=deform_knee_out_r}}
     {targetBone {boneName=deform_knee_in_r}}
     {targetBone {boneName=calf_twist_02_r}}
     {targetBone {boneName=thigh_twist_01_r}}
    ]
    constraintWeights [1,1,1,1,1,1]
   }
   Output Pose(Result=Component To Local(ComponentPose=@n5)) {layerGroup=None}
  ]}
 EventGraph {nodes [Event BlueprintUpdateAnimation()]}
]
variables []
