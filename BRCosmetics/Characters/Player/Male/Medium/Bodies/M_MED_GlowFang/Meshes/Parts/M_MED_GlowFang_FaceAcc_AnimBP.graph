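# Post-copy AnimBlueprint for the GlowFang male face accessory cosmetic, keeping it posed with the character.
# The base pose copies from the attached parent mesh.
# A layered per-bone blend overlays a second copy from the body part mesh on clavicle, upper arm, pec, hood and collar branches.
# OnCharacterCustomizationLoaded caches that body mesh from the owning PlayerPawn_Athena by part type.
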AnimBlueprint /BRCosmetics/Characters/Player/Male/Medium/Bodies/M_MED_GlowFang/Meshes/Parts/M_MED_GlowFang_FaceAcc_AnimBP
target_skeleton=&/M_MED_GlowFang_FaceAcc_Skeleton
graphs [
 AnimGraph {nodes [
   n1 = Layered blend per bone(
    BasePose=Copy Pose From Mesh() {bUseAttachedParent=true}
    BlendPoses_0=Copy Pose From Mesh(SourceMeshComponent=Get Body_Mesh()) {bUseAttachedParent=true}
   ) {layerSetup [{branchFilters [
       {boneName=clavicle_l}
       {boneName=clavicle_r}
       {boneName=upperarm_r}
       {boneName=upperarm_l}
       {boneName=deform_pec_l}
       {boneName=deform_pec_r}
       {boneName=BASE_dyn_hood_l}
       {boneName=BASE_dyn_hood_r}
       {boneName=dyn_hood_l}
       {boneName=dyn_hood_r}
       {boneName=dyn_collar_b_r}
       {boneName=dyn_collar_r}
       {boneName=dyn_collar_f_r}
       {boneName=dyn_collar_bk}
       {boneName=dyn_collar_b_l}
       {boneName=dyn_collar_l}
       {boneName=dyn_collar_f_l}
      ]}]}
   Output Pose(Result=@n1) {layerGroup=None}
  ]}
 EventGraph {nodes [
   Event OnCharacterCustomizationLoaded()
   > n1 = Cast To PlayerPawn_Athena(Object=TryGetPawnOwner())
   > Set Body_Mesh(Body_Mesh=GetSkeletalMeshForPartType(self=@n1))
   Event BlueprintUpdateAnimation()
  ]}
]
variables [Body_Mesh {type=SkeletalMeshComponent}]
